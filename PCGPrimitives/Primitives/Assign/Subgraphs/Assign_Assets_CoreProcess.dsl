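; PCG subgraph that assigns assets to points, with each spline or polygon input collapsed into one combined point.
; Assets are picked by per-data-set seed, index attribute or existing mesh attribute, optionally scaled to fit input bounds (per-axis or uniform max-dimension).
; Fixed and random transforms follow, then points are partitioned by asset to expand PCG data asset assemblies via Copy Points.
; Output bounds optionally come from meshes, with extents multiplied by ExtentsScale.
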(graph Assign_Assets_CoreProcess
  (params (PositionOffset :type Vector) (RotationOffset :type Rotator) (ScaleOffset :type Vector :default (1.0 1.0 1.0)) (ExtentsScale :type Vector :default (1.0 1.0 1.0)) (UseMeshBounds :type Bool :default true) (MatchScaleToInputBounds :type Bool) (UseMeshAttribute :type Bool) (MeshAttrinute :type String) (RandomPositionRange :type Vector) (RandomRotationRange :type Vector) (RandomScaleRange :type Vector) (Seed :type Int) (uniform_scale :type Bool :default true) (Use_Index_attribute :type Bool) (Index_attribute :type String))
  (bind AttributeRename_0 (Attribute%20Rename :AttributeToRename Assets :NewAttributeName Asset :In Assets))
  (bind Branch_1_Index (Branch :bOutputToB-param Use_Index_attribute)) ; Two asset selection modes:\n- Output A: seed-based random assignment\n- Output B: index attribute driven\n  (deterministic, user-controlled)\nControlled by Use Index Attribute.
  (bind MatchAndSetAttributes_0 (Match%20And%20Set%20Attributes))
  (bind CopyAttributes_1 (Copy%20Attributes :InputSource $Index :OutputTarget ind))
  (bind MatchAndSetAttributes_1 (Match%20And%20Set%20Attributes :bMatchAttributes true :InputAttribute $Seed :MatchAttribute ind :InputAttribute-param Index_attribute))
  (bind AttributeRename_2 (Attribute%20Rename :AttributeToRename Mesh :NewAttributeName Asset :AttributeToRename-param MeshAttrinute))
  (bind Select_1 (Select :bUseInputB-param UseMeshAttribute))
  (bind Branch_1 (Branch :bOutputToB-param MatchScaleToInputBounds))
  (bind AttributeMathsOp_5 (Attribute%20Maths%20Op :InputSource1 $Scale :InputSource2 Scale :Operation Set))
  (bind FilterDataByType_4 (Filter%20Data%20By%20Type :TargetType (struct :Ids [(struct :Struct /Script/PCG.PCGDataTypeInfoPoint)])))
  (bind TransformPoints_1 (Transform%20Points :bUniformScale-param uniform_scale :OffsetMax-param PositionOffset :OffsetMin-param PositionOffset :RotationMax-param RotationOffset :RotationMin-param RotationOffset :ScaleMax-param ScaleOffset :ScaleMin-param ScaleOffset))
  (bind BoundsFromMesh_1 (Bounds%20From%20Mesh :MeshAttribute Asset))
  (bind AttributeMathsOp_14 (Attribute%20Maths%20Op :InputSource1 $Extents :Operation Multiply :InB ExtentsScale))
  ;[0.15 0.15 0.15 0.5] DIFFERENT treatment for different types
  (bind FilterDataByType_5 (Filter%20Data%20By%20Type :TargetType (struct :Ids [(struct :Struct /Script/PCG.PCGDataTypeInfoPoint)])))
  ;;[0.15 0.15 0.15 0.5] treating spline and polygons as one entry
  (bind FilterDataByType_8 (Filter%20Data%20By%20Type :TargetType (struct :Ids [(struct :Struct /Script/PCG.PCGDataTypeInfoSpline)])))
  (bind FilterDataByType_9 (Filter%20Data%20By%20Type :TargetType (struct :Ids [(struct :Struct /Script/PCG.PCGDataTypeInfoPolygon2D)])))
  (bind GetSplineControlPoints_1 (Get%20Spline%20Control%20Points))
  (bind MergePoints_1 (Merge%20Points))
  (bind CombinePoints_0 (Combine%20Points))
  ;[0.15 0.15 0.15 0.5] seed per data stream
  (bind WriteDataIndex_1 (Write%20Data%20Index :bAddAttribute true :IndexAttribute dataIndex))
  (bind AttributeMathsOp_13 (Attribute%20Maths%20Op :InputSource1 $Seed :InputSource2 dataIndex :OutputTarget $Seed)) ; Offset the seed by the data partition index\nso each collection of points gets a unique\nrandom seed, avoiding identical results\nacross multiple input data sets.
  (bind MergePoints_0 (Merge%20Points))
  ;[0.15 0.15 0.15 0.5] SOLVING scale, the mesh size will match the input point size
  (bind FilterDataByType_3 (Filter%20Data%20By%20Type :TargetType (struct :Ids [(struct :Struct /Script/PCG.PCGDataTypeInfoPoint)])))
  (bind BoundsFromMesh_0 (Bounds%20From%20Mesh :MeshAttribute Asset))
  (bind AttributeMathsOp_6 (Attribute%20Maths%20Op :InputSource1 $BoundsMax :InputSource2 $BoundsMin :Operation Subtract :OutputTarget MeshSize))
  (bind AttributeMathsOp_7 (Attribute%20Maths%20Op :InputSource1 $BoundsMax :InputSource2 $BoundsMin :Operation Subtract :OutputTarget Size))
  (bind CopyAttributes_0 (Copy%20Attributes :InputSource MeshSize :OutputTarget MeshSize))
  (bind Branch_2 (Branch :bOutputToB-param uniform_scale))
  (bind AttributeMathsOp_4 (Attribute%20Maths%20Op :InputSource1 SizeMax :InputSource2 MeshSizeMax :Operation Divide :OutputTarget Scale))
  (bind AttributeMathsOp_16 (Attribute%20Maths%20Op :InputSource1 SizeMax :InputSource2 MeshSizeMax :Operation Divide :OutputTarget Scale))
  ;;[0.15 0.15 0.15 0.5] MAX input bounds size
  (bind ExtractAttribute_1 (Extract%20Attribute%20at%20Index :bForceOutputAttributeToBeInElementsDomain true :InputSource Size.X))
  (bind ExtractAttribute_2 (Extract%20Attribute%20at%20Index :bForceOutputAttributeToBeInElementsDomain true :InputSource Size.Y))
  (bind AttributeMathsOp_0 (Attribute%20Maths%20Op :Operation Max))
  (bind ExtractAttribute_3 (Extract%20Attribute%20at%20Index :bForceOutputAttributeToBeInElementsDomain true :InputSource Size.Z))
  (bind AttributeMathsOp_1 (Attribute%20Maths%20Op :Operation Max :OutputTarget SizeMax))
  ;;[0.15 0.15 0.15 0.5] Max mesh reference Size
  (bind ExtractAttribute_4 (Extract%20Attribute%20at%20Index :bForceOutputAttributeToBeInElementsDomain true :InputSource MeshSize.X))
  (bind ExtractAttribute_5 (Extract%20Attribute%20at%20Index :bForceOutputAttributeToBeInElementsDomain true :InputSource MeshSize.Y))
  (bind AttributeMathsOp_2 (Attribute%20Maths%20Op :Operation Max))
  (bind ExtractAttribute_6 (Extract%20Attribute%20at%20Index :bForceOutputAttributeToBeInElementsDomain true :InputSource MeshSize.Z))
  (bind AttributeMathsOp_3 (Attribute%20Maths%20Op :Operation Max :OutputTarget MeshSizeMax))
  ;[0.15 0.15 0.15 0.5] RANDOMIZE TRANSFORM
  (bind AttributeMathsOp_8 (Attribute%20Maths%20Op :Operation Multiply :InA RandomPositionRange))
  (bind AttributeMathsOp_10 (Attribute%20Maths%20Op :InA RandomScaleRange))
  (bind CreateConstant_0 (Create%20Constant :AttributeTypes (struct :Type Vector :DoubleValue 1.0 :VectorValue (1.0 1.0 1.0))))
  (bind AttributeMathsOp_11 (Attribute%20Maths%20Op :Operation Subtract :InB RandomScaleRange))
  (bind BreakVectorAttribute_0 (Break%20Vector%20Attribute :In RandomRotationRange))
  (bind MakeRotatorAttribute_0 (Make%20Rotator%20Attribute :Operation MakeRotFromAngles))
  (bind AttributeMathsOp_9 (Attribute%20Maths%20Op :Operation Multiply :InA RandomRotationRange))
  (bind BreakVectorAttribute_1 (Break%20Vector%20Attribute))
  (bind MakeRotatorAttribute_1 (Make%20Rotator%20Attribute :Operation MakeRotFromAngles))
  (bind TransformPoints_2 (Transform%20Points :OffsetMax (-3e+03 0.0 0.0) :OffsetMin (3e+03 0.0 0.0) :bUniformScale-param uniform_scale :OffsetMax-param RandomPositionRange :Seed-param Seed))
  ;[0.15 0.15 0.15 0.5] ASSEMBLIES
  (bind AttributePartition_1 (Attribute%20Partition :PartitionAttributeSelectors [Asset]))
  (bind ExtractAttribute_0 (Extract%20Attribute%20at%20Index :bForceOutputAttributeToBeInElementsDomain true :InputSource Asset))
  (bind PCGLoadDataAsset_0 (Load%20PCG%20Data%20Asset :bLoadFromInput true))
  (bind ToPoint_0 (To%20Point))
  (bind FilterDataByTag_1 (Filter%20Data%20By%20Tag :SelectedTags "points"))
  (bind AttributeRename_1 (Attribute%20Rename :AttributeToRename Mesh :NewAttributeName Asset))
  (bind CopyPoints_0 (Copy%20Points :bCopyEachSourceOnEveryTarget false))
  (bind Branch_0 (Branch :bOutputToB-param UseMeshBounds))
  (connect ExtractAttribute_1 Out AttributeMathsOp_0 InA)
  (connect ExtractAttribute_2 Out AttributeMathsOp_0 InB)
  (connect AttributeMathsOp_0 Out AttributeMathsOp_1 InA)
  (connect ExtractAttribute_3 Out AttributeMathsOp_1 InB)
  (connect ExtractAttribute_4 Out AttributeMathsOp_2 InA)
  (connect ExtractAttribute_5 Out AttributeMathsOp_2 InB)
  (connect AttributeMathsOp_2 Out AttributeMathsOp_3 InA)
  (connect ExtractAttribute_6 Out AttributeMathsOp_3 InB)
  (connect Branch_2 Output%20A AttributeMathsOp_4 InA)
  (connect Branch_2 Output%20A AttributeMathsOp_4 InB)
  (connect Branch_1 Output%20B AttributeMathsOp_5 InA)
  (connect AttributeMathsOp_4 Out AttributeMathsOp_5 InB)
  (connect AttributeMathsOp_16 Out AttributeMathsOp_5 InB)
  (connect BoundsFromMesh_0 Out AttributeMathsOp_6 InA)
  (connect BoundsFromMesh_0 Out AttributeMathsOp_6 InB)
  (connect Branch_1 Output%20B AttributeMathsOp_7 InA)
  (connect Branch_1 Output%20B AttributeMathsOp_7 InB)
  (connect CreateConstant_0 Out AttributeMathsOp_11 InA)
  (connect WriteDataIndex_1 Out AttributeMathsOp_13 InA)
  (connect WriteDataIndex_1 Out AttributeMathsOp_13 InB)
  (connect BoundsFromMesh_1 Out AttributeMathsOp_14 InA)
  (connect Branch_0 Output%20A AttributeMathsOp_14 InA)
  (connect AttributeMathsOp_1 Out AttributeMathsOp_16 InA)
  (connect AttributeMathsOp_3 Out AttributeMathsOp_16 InB)
  (connect TransformPoints_2 Out AttributePartition_1 In)
  (connect FilterDataByTag_1 InsideFilter AttributeRename_1 In)
  (connect MatchAndSetAttributes_0 Out AttributeRename_2 In)
  (connect MatchAndSetAttributes_1 Out AttributeRename_2 In)
  (connect FilterDataByType_3 InsideFilter BoundsFromMesh_0 In)
  (connect Branch_0 Output%20B BoundsFromMesh_1 In)
  (connect CopyPoints_0 Out Branch_0 In)
  (connect TransformPoints_2 Out Branch_0 In)
  (connect Select_1 Out Branch_1 In)
  (connect CopyAttributes_0 Out Branch_2 In)
  (connect MergePoints_0 Out Branch_1_Index In)
  (connect AttributeMathsOp_9 Out BreakVectorAttribute_1 In)
  (connect MergePoints_1 Out CombinePoints_0 In)
  (connect AttributeMathsOp_6 Out CopyAttributes_0 Source)
  (connect AttributeMathsOp_7 Out CopyAttributes_0 Target)
  (connect AttributeRename_0 Out CopyAttributes_1 Source)
  (connect AttributeRename_0 Out CopyAttributes_1 Target)
  (connect AttributeRename_1 Out CopyPoints_0 Source)
  (connect AttributePartition_1 Out CopyPoints_0 Target)
  (connect AttributePartition_1 Out ExtractAttribute_0 In)
  (connect Branch_2 Output%20B ExtractAttribute_1 In)
  (connect Branch_2 Output%20B ExtractAttribute_2 In)
  (connect Branch_2 Output%20B ExtractAttribute_3 In)
  (connect Branch_2 Output%20B ExtractAttribute_4 In)
  (connect Branch_2 Output%20B ExtractAttribute_5 In)
  (connect Branch_2 Output%20B ExtractAttribute_6 In)
  (connect ToPoint_0 Out FilterDataByTag_1 In)
  (connect Branch_1 Output%20B FilterDataByType_3 In)
  (connect AttributeMathsOp_5 Out FilterDataByType_4 In)
  (connect Branch_1 Output%20A FilterDataByType_4 In)
  (connect DefaultInputNode In FilterDataByType_5 In)
  (connect DefaultInputNode In FilterDataByType_8 In)
  (connect DefaultInputNode In FilterDataByType_9 In)
  (connect FilterDataByType_8 InsideFilter GetSplineControlPoints_1 In)
  (connect FilterDataByType_9 InsideFilter GetSplineControlPoints_1 In)
  (connect BreakVectorAttribute_0 Y MakeRotatorAttribute_0 Pitch)
  (connect BreakVectorAttribute_0 X MakeRotatorAttribute_0 Roll)
  (connect BreakVectorAttribute_0 Z MakeRotatorAttribute_0 Yaw)
  (connect BreakVectorAttribute_1 Y MakeRotatorAttribute_1 Pitch)
  (connect BreakVectorAttribute_1 X MakeRotatorAttribute_1 Roll)
  (connect BreakVectorAttribute_1 Z MakeRotatorAttribute_1 Yaw)
  (connect Branch_1_Index Output%20A MatchAndSetAttributes_0 In)
  (connect AttributeRename_0 Out MatchAndSetAttributes_0 Match%20Data)
  (connect Branch_1_Index Output%20B MatchAndSetAttributes_1 In)
  (connect CopyAttributes_1 Out MatchAndSetAttributes_1 Match%20Data)
  (connect AttributeMathsOp_13 Out MergePoints_0 In)
  (connect GetSplineControlPoints_1 Out MergePoints_1 In)
  (connect ExtractAttribute_0 Out PCGLoadDataAsset_0 In)
  (connect MatchAndSetAttributes_0 Out Select_1 Input%20A)
  (connect MatchAndSetAttributes_1 Out Select_1 Input%20A)
  (connect AttributeRename_2 Out Select_1 Input%20B)
  (connect PCGLoadDataAsset_0 Out ToPoint_0 In)
  (connect FilterDataByType_4 InsideFilter TransformPoints_1 In)
  (connect TransformPoints_1 Out TransformPoints_2 In)
  (connect AttributeMathsOp_8 Out TransformPoints_2 OffsetMin)
  (connect MakeRotatorAttribute_1 Out TransformPoints_2 RotationMax)
  (connect MakeRotatorAttribute_0 Out TransformPoints_2 RotationMin)
  (connect AttributeMathsOp_10 Out TransformPoints_2 ScaleMax)
  (connect AttributeMathsOp_11 Out TransformPoints_2 ScaleMin)
  (connect CombinePoints_0 Out WriteDataIndex_1 In)
  (connect FilterDataByType_5 InsideFilter WriteDataIndex_1 In)
  (output Out AttributeMathsOp_14))
