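; PCG subgraph that runs the Assign rule's core pass, looping an inner subgraph over the inputs with grammar, module infos, random transform and mesh override settings.
; It also builds footprints from the widest X extent among the loop's point output.
; Open splines are trimmed by that extent, then join closed ones as force-open polygons twice as wide.
; Footprints are tagged TEMP_Footprint and output alongside the loop results.
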(graph Assign_Rule_CoreProcess
  (params (Grammar :type String) (ModulesInfos :type (Array (Struct /PCGPrimitives/Grammar/Structure/STC_Module.STC_Module))) (RandPosRange :type Vector) (RandRotRange :type Vector) (RandScaleRange :type Vector) (PosOffset :type Vector) (OverrideScale :type Vector) (MeshOverride :type (SoftObject /Script/Engine.StaticMesh)) (SpawnAssets :type Bool))
  (inputs (Modules_infos :type Param))
  (bind Loop_1 (Loop :Graph /PCGPrimitives/Primitives/Assign/Subgraphs/Assign_Rule_CoreProcess_InnerLoop.Assign_Rule_CoreProcess_InnerLoop :Grammar Grammar :MeshOverride MeshOverride :OverrideScale OverrideScale :PosOffset PosOffset :RandPosRange RandPosRange :RandRotRange RandRotRange :RandScaleRange RandScaleRange :SpawnAssets SpawnAssets))
  ;[0.15 0.15 0.15 0.5] FOOTPRINT BUILDING, OUT of WIDEST BOUND on X
  (bind FilterDataByType_1 (Filter%20Data%20By%20Type :TargetType (struct :Ids [(struct :Struct /Script/PCG.PCGDataTypeInfoPoint)])))
  (bind MergePoints_0 (Merge%20Points))
  (bind AttributeReduce_0 (Attribute%20Reduce :InputSource $Extents.X :Operation Max)) ; Find widest module (max X extent)
  (bind AttributeMathsOp_0 (Attribute%20Maths%20Op :Operation Multiply)) ; Double extent -> full strip width across spline
  (bind FilterDataByAttribute_1 (Filter%20Data%20By%20Attribute :FilterMode FilterByValue :FilterOperator Equal :TargetAttribute @data.$IsClosed :Threshold (struct :AttributeTypes (struct :Type Boolean)))) ; Open splines (paths) vs closed splines (areas)
  (bind FilterDataByType_0 (Filter%20Data%20By%20Type :TargetType (struct :Ids [(struct :Struct /Script/PCG.PCGDataTypeInfoSpline)] :CustomSubtype 0)))
  (bind Subgraph_0 (Subgraph :Graph /PCGPrimitives/Primitives/Shared/Trim_open_Spline_each_side.Trim_open_Spline_each_side)) ; Trim open splines so footprint doesn't\noverhang module placement
  (bind CreatePolygon2D_1 (Create%20Polygon%202D :InputType ForceOpen))
  (bind AddTags_0 (Add%20Tags :TagsToAdd "TEMP_Footprint")) ; TEMP tag - PrepareOutputs routes this\nto the Footprints output
  (connect CreatePolygon2D_1 Out AddTags_0 In)
  (connect AttributeReduce_0 Out AttributeMathsOp_0 InA)
  (connect MergePoints_0 Out AttributeReduce_0 In)
  (connect FilterDataByAttribute_1 OutsideFilter CreatePolygon2D_1 In)
  (connect Subgraph_0 Out CreatePolygon2D_1 In)
  (connect AttributeMathsOp_0 Out CreatePolygon2D_1 OpenPolygonWidth)
  (connect DefaultInputNode In FilterDataByAttribute_1 In)
  (connect FilterDataByAttribute_1 InsideFilter FilterDataByType_0 In)
  (connect Loop_1 Out FilterDataByType_1 In)
  (connect DefaultInputNode In Loop_1 In)
  (connect DefaultInputNode Modules_infos Loop_1 Modules_infos)
  (connect FilterDataByType_1 InsideFilter MergePoints_0 In)
  (connect AttributeReduce_0 Out Subgraph_0 distance)
  (connect FilterDataByType_0 InsideFilter Subgraph_0 In)
  (output Out AddTags_0)
  (output Out Loop_1))
