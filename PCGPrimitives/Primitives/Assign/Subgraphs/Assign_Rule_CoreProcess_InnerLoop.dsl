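; PCG loop-body subgraph for an assign rule that runs the ShapeGrammar_Module subgraph on incoming splines with a given grammar, forwarding offset, random transform, scale and spawn settings.
; Splines are filtered from the input and tagged with a default slot scale, as this rule level has no per-slot override.
; Module info attributes are renamed, and MeshOverride replaces module meshes when set.
; The loop index seeds randomness, varying each iteration.
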(graph Assign_Rule_CoreProcess_InnerLoop
  (params (Grammar :type String) (ModulesInfos :type (Array (Struct /PCGPrimitives/Grammar/Structure/STC_Module.STC_Module))) (RandPosRange :type Vector) (RandRotRange :type Vector) (RandScaleRange :type Vector) (PosOffset :type Vector) (OverrideScale :type Vector) (MeshOverride :type (SoftObject /Script/Engine.StaticMesh)) (SpawnAssets :type Bool))
  (inputs (Modules_infos :type Param))
  (bind FilterDataByType_0 (Filter%20Data%20By%20Type :TargetType (struct :Ids [(struct :Struct /Script/PCG.PCGDataTypeInfoSpline)] :CustomSubtype 0)))
  (bind AddAttribute_0 (Add%20Attribute :AttributeTypes (struct :DoubleValue 1.0) :OutputTarget @data.slotScale)) ; Default slot scale = 1\n(no per-slot override at rule level)
  (bind CreateConstant_0 (Create%20Constant :AttributeTypes (struct :Type Boolean)))
  (bind GetLoopIndex_0 (Get%20Loop%20Index)) ; Loop index = seed (varies RNG per iteration)
  (bind CreateConstant_1 (Create%20Constant :AttributeTypes (struct :Type String)))
  (bind AttributeCompareOp_0 (Attribute%20Compare%20Op :InA MeshOverride))
  (bind AttributeRename_0 (Attribute%20Rename :AttributeToRename Mesh :NewAttributeName Module%20Mesh))
  (bind AttributeRename_1 (Attribute%20Rename :AttributeToRename Size :NewAttributeName Module%20Size))
  (bind CopyAttributes_0 (Copy%20Attributes :InputSource @Last :OutputTarget "Module Mesh" :Source MeshOverride))
  (bind Select_0 (Select)) ; If MeshOverride is set, use it;\notherwise fall back to module-defined mesh
  (bind Subgraph_0 (Subgraph :Graph /PCGPrimitives/Grammar/PCG/ShapeGrammar_Module.ShapeGrammar_Module :Grammar Grammar :Pos_offset PosOffset :Random_Position_Range RandPosRange :Random_Rotation_Range RandRotRange :Random_Scale_Range RandScaleRange :Scale_override OverrideScale :Spawn_Assets SpawnAssets))
  (connect FilterDataByType_0 InsideFilter AddAttribute_0 In)
  (connect CreateConstant_1 Out AttributeCompareOp_0 InB)
  (connect DefaultInputNode Modules_infos AttributeRename_0 In)
  (connect AttributeRename_0 Out AttributeRename_1 In)
  (connect AttributeRename_1 Out CopyAttributes_0 Target)
  (connect DefaultInputNode In FilterDataByType_0 In)
  (connect AttributeCompareOp_0 Out Select_0 bUseInputB)
  (connect CopyAttributes_0 Out Select_0 Input%20A)
  (connect AttributeRename_1 Out Select_0 Input%20B)
  (connect CreateConstant_0 Out Subgraph_0 DebugSegments)
  (connect Select_0 Out Subgraph_0 modules%20infos)
  (connect GetLoopIndex_0 Out Subgraph_0 Seed)
  (connect AddAttribute_0 Out Subgraph_0 Splines)
  (output Out Subgraph_0.output))
