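; PCG core subgraph that applies a Shape Grammar Definition asset to input splines, outputting slot results and footprints.
; Get_SGD_Data converts the asset into an attribute set, feeding Slots_Wrapper, or an inner loop when UseAttributeForSize is set.
; On the wrapper path, the asset's orientation selects TEMP_Footprint shapes: splines offset upward by slot height, or Size-based polygons, ringed for closed splines.
; Spline roll is reset first; the Blueprint reader remains as unused reference.
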(graph Assign_ShapeGrammarDefinition_CoreProcess
  (params (ShapeGrammarDefinition :type (SoftObject /PCGPrimitives/Grammar/BP/BP_ShapeGrammarDefinition.BP_ShapeGrammarDefinition_C)) (Size :type Double :default 1e+03) (UseAttributeForSize :type String) (SpawnAssets :type Bool :default true))
  (bind FilterDataByType_0 (Filter%20Data%20By%20Type :TargetType (struct :Ids [(struct :Struct /Script/PCG.PCGDataTypeInfoSpline)])))
  (bind AttributeMathsOp_0 (Attribute%20Maths%20Op :InputSource1 $Rotation.Roll :Operation Set)) ; Forcing Spline point facing up
  (bind CreateConstant_0 (Create%20Constant :AttributeTypes (struct :Type String)))
  (bind AttributeCompareOp_0 (Attribute%20Compare%20Op :InA UseAttributeForSize))
  (bind Branch_1 (Branch))
  ;[0 0 0 0.5] FOOTPRINT
  (bind GetPropertyFromObjectPath_0 (Get%20Property%20From%20Object%20Path :PropertyName orientation :In ShapeGrammarDefinition))
  (bind Branch_0 (Branch))
  (bind FilterDataByAttribute_1 (Filter%20Data%20By%20Attribute :FilterMode FilterByValue :FilterOperator Equal :TargetAttribute @data.$IsClosed :Threshold (struct :AttributeTypes (struct :Type Boolean))))
  (bind AddAttribute_0 (Add%20Attribute :AttributeTypes (struct :Type Vector :VectorValue (0.0 0.0 1.0)) :OutputTarget Direction))
  (bind MergePoints_0 (Merge%20Points))
  (bind AttributePartition_0 (Attribute%20Partition :PartitionAttributeSelectors [SplineIndex]))
  (bind CombinePoints_0 (Combine%20Points))
  (bind AttributeMathsOp_3 (Attribute%20Maths%20Op :InputSource1 $BoundsMax.Z :InputSource2 $BoundsMin.Z :Operation Subtract :OutputTarget Size))
  (bind ExtractAttribute_0 (Extract%20Attribute%20at%20Index :bForceOutputAttributeToBeInElementsDomain true :InputSource Size))
  (bind OffsetSpline_0 (Offset%20Spline))
  (bind AddTags_1 (Add%20Tags :TagsToAdd "TEMP_Footprint"))
  ;;[1 0 0 0.5] Clean up resampled
  (bind CleanSpline_1 (Clean%20Spline :CollinearAngleThreshold 1.0))
  ;;[0 0 0 0.5] FOOTPRINT
  (bind CreatePolygon2D_1 (Create%20Polygon%202D :InputType ForceOpen :OpenPolygonWidth-param Size))
  ;;[0 0 0 0.5] If closed spline
  (bind AttributeMathsOp_1 (Attribute%20Maths%20Op :Operation Multiply :InA Size))
  (bind CreatePolygon2D_0 (Create%20Polygon%202D :InputType ForceClosed))
  (bind OffsetPolygon2D_1 (Offset%20Polygon))
  (bind AttributeMathsOp_2 (Attribute%20Maths%20Op :Operation Multiply :InA Size))
  (bind OffsetPolygon2D_2 (Offset%20Polygon))
  (bind Polygon2DOperation_1 (Polygon%20Operation :Operation Difference))
  ;[0 0 0 0.5] Core Process
  (bind Subgraph_0 (Subgraph :Graph /PCGPrimitives/Grammar/PCG/Slots_Wrapper.Slots_Wrapper :ShapeGrammarDefinition ShapeGrammarDefinition :Size Size :SpawnAssets SpawnAssets :UseAttributeForSize UseAttributeForSize))
  (bind Loop_0 (Loop :Graph /PCGPrimitives/Primitives/Assign/Subgraphs/Assign_ShapeGrammarDefinition_CoreProcess_innerLoop.Assign_ShapeGrammarDefinition_CoreProcess_innerLoop :ShapeGRammarDefinition ShapeGrammarDefinition :UseAttributeForSize UseAttributeForSize))
  ;[0 0 0 0.5 32] Extract Shape Grammar Definition Data
  (bind Subgraph_1 (Subgraph :Graph /PCGPrimitives/Grammar/PCG/Get_SGD_Data.Get_SGD_Data :ShapeGrammarDefinition ShapeGrammarDefinition)) ; EXTRACT all parameters from the asset inputed \nand convert them to a giant attribute set
  ;;[0.1 0 0 1 12] Deprecated BP method, un-used
  (bind ExecuteBlueprint_0 (Execute%20Blueprint :BlueprintElementInstance (class:/PCGPrimitives/Grammar/BP/BP_Read_ShapeGrammarDefinition.BP_Read_ShapeGrammarDefinition_C :Category "Metadata") :BlueprintElementType /PCGPrimitives/Grammar/BP/BP_Read_ShapeGrammarDefinition.BP_Read_ShapeGrammarDefinition_C :Shape%20Grammar%20Definition ShapeGrammarDefinition)) ; Custom BP Element extraction reference/backup \n(replaced by PCG subgraph above)
  (reroute SGD_Data Subgraph_1.SGD_Data)
  (connect Branch_0 Output%20A AddAttribute_0 In)
  (connect CreatePolygon2D_1 Out AddTags_1 In)
  (connect OffsetSpline_0 Out AddTags_1 In)
  (connect Polygon2DOperation_1 Out AddTags_1 In)
  (connect CreateConstant_0 Out AttributeCompareOp_0 InB)
  (connect FilterDataByType_0 InsideFilter AttributeMathsOp_0 InA)
  (connect CombinePoints_0 Out AttributeMathsOp_3 InA)
  (connect CombinePoints_0 Out AttributeMathsOp_3 InB)
  (connect MergePoints_0 Out AttributePartition_0 In)
  (connect GetPropertyFromObjectPath_0 Out Branch_0 bOutputToB)
  (connect Branch_1 Output%20B Branch_0 In)
  (connect AttributeCompareOp_0 Out Branch_1 bOutputToB)
  (connect AttributeMathsOp_0 Out Branch_1 In)
  (connect FilterDataByAttribute_1 InsideFilter CleanSpline_1 In)
  (connect AttributePartition_0 Out CombinePoints_0 In)
  (connect FilterDataByAttribute_1 OutsideFilter CreatePolygon2D_0 In)
  (connect CleanSpline_1 Out CreatePolygon2D_1 In)
  (connect AttributeMathsOp_3 Out ExtractAttribute_0 In)
  (connect Branch_0 Output%20B FilterDataByAttribute_1 In)
  (connect DefaultInputNode In FilterDataByType_0 In)
  (connect Branch_1 Output%20A Loop_0 In)
  (connect SGD_Data Out Loop_0 SGD_Data)
  (connect Subgraph_0 Out MergePoints_0 In)
  (connect CreatePolygon2D_0 Out OffsetPolygon2D_1 In)
  (connect AttributeMathsOp_1 Out OffsetPolygon2D_1 Offset)
  (connect CreatePolygon2D_0 Out OffsetPolygon2D_2 In)
  (connect AttributeMathsOp_2 Out OffsetPolygon2D_2 Offset)
  (connect AddAttribute_0 Out OffsetSpline_0 In)
  (connect ExtractAttribute_0 Out OffsetSpline_0 Magnitude)
  (connect OffsetPolygon2D_2 Out Polygon2DOperation_1 ClipPolygons)
  (connect OffsetPolygon2D_1 Out Polygon2DOperation_1 In)
  (connect Branch_1 Output%20B Subgraph_0 In)
  (connect SGD_Data Out Subgraph_0 SGD_Data)
  (output Out AddTags_1)
  (output Out Loop_0)
  (output Out Subgraph_0))
