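; PCG inner-loop subgraph that runs Slots_Wrapper for one shape grammar definition and outputs its results alongside a footprint tagged TEMP_Footprint.
; An orientation property selects the footprint method: splines offset using a Z Direction and the height of combined slot bounds, or Size-driven polygons.
; In the polygon branch, open splines become width-based polygons, while closed ones become bands from the difference of two offset polygons.
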(graph Assign_ShapeGrammarDefinition_CoreProcess_innerLoop
  (params (ShapeGRammarDefinition :type (SoftObject /PCGPrimitives/Grammar/BP/BP_ShapeGrammarDefinition.BP_ShapeGrammarDefinition_C)) (Size :type Double :default 1e+03) (UseAttributeForSize :type String))
  (inputs (SGD_Data :type Param))
  (bind Subgraph_1 (Subgraph :Graph /PCGPrimitives/Grammar/PCG/Slots_Wrapper.Slots_Wrapper :ShapeGrammarDefinition ShapeGRammarDefinition :Size Size :UseAttributeForSize UseAttributeForSize))
  ;[0 0 0 0.5] FOOTPRINT
  (bind GetPropertyFromObjectPath_1 (Get%20Property%20From%20Object%20Path :PropertyName orientation))
  (bind Branch_1 (Branch))
  (bind FilterDataByAttribute_0 (Filter%20Data%20By%20Attribute :FilterMode FilterByValue :FilterOperator Equal :TargetAttribute @data.$IsClosed :Threshold (struct :AttributeTypes (struct :Type Boolean))))
  (bind CreatePolygon2D_0 (Create%20Polygon%202D :InputType ForceOpen :OpenPolygonWidth-param Size))
  (bind AddAttribute_1 (Add%20Attribute :AttributeTypes (struct :Type Vector :VectorValue (0.0 0.0 1.0)) :OutputTarget Direction))
  (bind MergePoints_1 (Merge%20Points))
  (bind AttributePartition_1 (Attribute%20Partition :PartitionAttributeSelectors [SplineIndex]))
  (bind CombinePoints_1 (Combine%20Points))
  (bind AttributeMathsOp_2 (Attribute%20Maths%20Op :InputSource1 $BoundsMax.Z :InputSource2 $BoundsMin.Z :Operation Subtract :OutputTarget Size))
  (bind ExtractAttribute_1 (Extract%20Attribute%20at%20Index :bForceOutputAttributeToBeInElementsDomain true :InputSource Size))
  (bind OffsetSpline_1 (Offset%20Spline))
  (bind AttributeMathsOp_0 (Attribute%20Maths%20Op :Operation Multiply :InA Size))
  (bind CreatePolygon2D_1 (Create%20Polygon%202D :InputType ForceClosed))
  (bind OffsetPolygon2D_0 (Offset%20Polygon))
  (bind AttributeMathsOp_1 (Attribute%20Maths%20Op :Operation Multiply :InA Size))
  (bind OffsetPolygon2D_1 (Offset%20Polygon))
  (bind Polygon2DOperation_0 (Polygon%20Operation :Operation Difference))
  (bind AddTags_0 (Add%20Tags :TagsToAdd "TEMP_Footprint"))
  (connect Branch_1 Output%20A AddAttribute_1 In)
  (connect CreatePolygon2D_0 Out AddTags_0 In)
  (connect OffsetSpline_1 Out AddTags_0 In)
  (connect Polygon2DOperation_0 Out AddTags_0 In)
  (connect CombinePoints_1 Out AttributeMathsOp_2 InA)
  (connect CombinePoints_1 Out AttributeMathsOp_2 InB)
  (connect MergePoints_1 Out AttributePartition_1 In)
  (connect GetPropertyFromObjectPath_1 Out Branch_1 bOutputToB)
  (connect DefaultInputNode In Branch_1 In)
  (connect AttributePartition_1 Out CombinePoints_1 In)
  (connect FilterDataByAttribute_0 InsideFilter CreatePolygon2D_0 In)
  (connect FilterDataByAttribute_0 OutsideFilter CreatePolygon2D_1 In)
  (connect AttributeMathsOp_2 Out ExtractAttribute_1 In)
  (connect Branch_1 Output%20B FilterDataByAttribute_0 In)
  (connect Subgraph_1 Out MergePoints_1 In)
  (connect CreatePolygon2D_1 Out OffsetPolygon2D_0 In)
  (connect AttributeMathsOp_0 Out OffsetPolygon2D_0 Offset)
  (connect CreatePolygon2D_1 Out OffsetPolygon2D_1 In)
  (connect AttributeMathsOp_1 Out OffsetPolygon2D_1 Offset)
  (connect AddAttribute_1 Out OffsetSpline_1 In)
  (connect ExtractAttribute_1 Out OffsetSpline_1 Magnitude)
  (connect OffsetPolygon2D_1 Out Polygon2DOperation_0 ClipPolygons)
  (connect OffsetPolygon2D_0 Out Polygon2DOperation_0 In)
  (connect DefaultInputNode In Subgraph_1 In)
  (connect DefaultInputNode SGD_Data Subgraph_1 SGD_Data)
  (output Out AddTags_0)
  (output Out Subgraph_1))
