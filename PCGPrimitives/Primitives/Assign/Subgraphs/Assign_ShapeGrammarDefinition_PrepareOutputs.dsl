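; PCG subgraph that prepares the Footprints and Points outputs of the Assign Shape Grammar Definition primitive.
; It drops empty data, then splits the input by the TEMP_Footprint tag into footprint polygons and grammar points, stripping that temporary tag from footprints.
; Each stream runs through a Data Count and Branch gate, reaching the shared OutputTaggingSG subgraph only when it holds data, to avoid empty-data issues.
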(graph Assign_ShapeGrammarDefinition_PrepareOutputs
  (outputs (Footprints :type Any) (Points :type Any))
  (bind RemoveEmptyData_0 (Remove%20Empty%20Data))
  ;[0 0 0 0.5] Tag and apply Metadata for Output
  (bind FilterDataByTag_0 (Filter%20Data%20By%20Tag :SelectedTags "TEMP_Footprint")) ; Separate footprint polygons (TEMP_Footprint)\nfrom grammar points
  (bind DataNum_0 (Data%20Count))
  (bind Branch_0 (Branch)) ; Skip output tagging when no points\n(avoids empty-data issues)
  (bind Subgraph_0 (Subgraph :Graph /PCGPrimitives/Primitives/Shared/PrepareOutputs/OutputTaggingSG.OutputTaggingSG))
  (bind DeleteTags_1 (Delete%20Tags :SelectedTags "TEMP_Footprint")) ; Strip TEMP tag before exposing\nfootprints externally
  (bind DataNum_1 (Data%20Count))
  (bind Branch_1 (Branch)) ; Skip output tagging when no footprints\n(avoids empty-data issues)
  (bind Subgraph_1 (Subgraph :Graph /PCGPrimitives/Primitives/Shared/PrepareOutputs/OutputTaggingSG.OutputTaggingSG))
  (connect DataNum_0 Out Branch_0 bOutputToB)
  (connect FilterDataByTag_0 OutsideFilter Branch_0 In)
  (connect DataNum_1 Out Branch_1 bOutputToB)
  (connect DeleteTags_1 Out Branch_1 In)
  (connect FilterDataByTag_0 OutsideFilter DataNum_0 In)
  (connect DeleteTags_1 Out DataNum_1 In)
  (connect FilterDataByTag_0 InsideFilter DeleteTags_1 In)
  (connect RemoveEmptyData_0 Out FilterDataByTag_0 In)
  (connect DefaultInputNode In RemoveEmptyData_0 In)
  (connect Branch_0 Output%20B Subgraph_0 In)
  (connect Branch_1 Output%20B Subgraph_1 In)
  (output Footprints Subgraph_1)
  (output Points Subgraph_0))
